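# Anim Blueprint for the CasinoReaper Die female body cosmetic, following the parent mesh's pose.
# The copied pose feeds a Control Rig, zipper AnimDynamics and an additive saber_base_r rotation while DBNO.
# A lobby curve weights a DefaultSlot blend on saber bones.
# Weapon, skydive, emote-end and customization events set material "Hide Element 10" on when a CasinoReaper-tagged pickaxe is equipped outside skydiving, off otherwise.
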AnimBlueprint /BRCosmetics/Characters/Player/Female/Medium/Bodies/F_MED_CasinoReaper_Die/Meshes/F_MED_CasinoReaper_Die_AnimBP
target_skeleton=&/F_MED_CasinoReaper_Die_Skeleton
graphs [
 AnimGraph {nodes [
   n1 = Control Rig!(
    Source=Copy Pose From Mesh() {bUseAttachedParent=true}
    Alpha
    Is Sliding:bool=Get bIsSliding()
    is crouch not moving:bool=Get Is crouching not moving() -> (Is crouching not moving:bool)
    is in FE:bool=Get bIsInFrontEnd()
   ) -> (Pose) {
    alphaBoolBlend {blendOption=Linear}
    controlRigAssetReference {blueprintRigClass=&/F_MED_CasinoReaper_Die_CtrlRig.F_MED_CasinoReaper_Die_CtrlRig_C}
   }
   n2 = Anim Dynamics - Bone: dyn_zipper(ComponentPose=Local To Component(LocalPose=@n1), Alpha=1) {
    angularDampingOverride=0.99
    bOverrideAngularDamping=true
    bUseGravityOverride=true
    alphaBoolBlend {blendOption=Linear}
    boundBone {boneName=dyn_zipper}
    gravityOverride {z=-980}
    physicsBodyDefinitions [{
      boundBone {boneName=dyn_zipper}
      boxExtents=<5 5 5>
      constraintSetup {angularLimitsMax=<0 5 30>, angularLimitsMin=<0 -5 -10>, angularTarget=<0 0 -1>}
      localJointOffset=<0.5 0 0>
     }]
   }
   n3 = "Transform (Modify) Bone - Bone: saber_base_r"!(ComponentPose=@n2, bAlphaBoolEnabled:bool=Get bIsDBNO(), Translation="0,0,0", Rotation="-12.639,-13.173,-43.806", Scale="1,1,1") -> (Pose) {
    alphaInputType=Bool
    rotationMode=BMM_Additive
    alphaBoolBlend {blendOption=Linear, blendOutTime=100}
    alphaScaleBiasClamp {bInterpResult=true}
    boneToModify {boneName=saber_base_r}
    rotation=<rot -12.639 -13.173 -43.806>
   }
   n4 = Layered blend per bone(
    BasePose=Component To Local(ComponentPose=@n3)
    BlendPoses_0=Slot 'DefaultSlot'()
    BlendWeights_0=Get FrontEndLobbySync() -> (FrontEndLobbySync:float)
   ) {layerSetup [{branchFilters [{boneName=saber_base_l}, {boneName=dyn_saber_l}, {boneName=saber_base_r}, {boneName=dyn_saber_r}]}]}
   Output Pose(Result=@n4) {layerGroup=None}
  ]}
 EventGraph {nodes [
   Event BlueprintUpdateAnimation()
   > n1 = Sequence()
   n2 = Get bIsMoving2D()
   > Set Is crouching not moving(Is crouching not moving:bool=AND Boolean(A=Get bIsCrouching(), B=NOT Boolean(A=@n2))) -> (Output_Get:bool)
   Branch(execute=@n1.then_1, Condition=Get bIsInFrontEnd())
   > Set FrontEndLobbySync(FrontEndLobbySync:float=GetCurveValue(self=GetOwnerFrontEndAnimInstance(), CurveName=Frontend_Idle_Bypass)) -> (Output_Get:float)
   n3 = Event OnCharacterCustomizationLoaded()
   n4 = Get Dynamic Material Instance(execute=[@n3.then, @n8.then, @n9.then, @n10.then, @n11.then], SkeletalMeshComponent=GetPartSkeletalMeshComponent())
   n5 = Get bIsSkyDiving()
   n6 = Make Array("[0]"=(TagName="Cosmetics.MetaTags.CasinoReaper"))
   > n7 = Branch(Condition=AND Boolean(A=PickaxeItemDefinitionHasAnyMetaTag(Tags=@n6, bRequiresEquipped=true), B=NOT Boolean(A=@n5)))
   > SetScalarParameterValue(self=@n4, ParameterName=Hide Element 10, Value=1)
   SetScalarParameterValue(execute=@n7.else, self=@n4, ParameterName=Hide Element 10, Value=0)
   n8 = Event OnEndPlayingEmote() {is=title K2Node_Event}
   n9 = Event OnEndSkydiving()
   n10 = Event OnBeginSkydiving()
   n11 = Event OnWeaponChanged()
  ]}
]
variables [
 FrontEndLobbySync {type=number, default=1}
 Is crouching not moving {type=boolean}
]
